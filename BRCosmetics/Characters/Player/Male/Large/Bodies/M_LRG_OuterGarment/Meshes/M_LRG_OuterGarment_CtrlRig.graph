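# Control Rig for the large male outer garment that corrects shoulder pad placement as the character moves.
# The forward solve copies upper-arm rotations to repeat bones, then offsets their child pads by smoothed parachute and skydiving weights.
# Spherical pose readers on the upper arms add corrective offsets, some left-side ones gated by is_melee or in_FE.
# Setup aligns two temporary controls to variable-named bones; their write-back chain is not connected to execution.
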ControlRigBlueprint /BRCosmetics/Characters/Player/Male/Large/Bodies/M_LRG_OuterGarment/Meshes/M_LRG_OuterGarment_CtrlRig
graphs [
 RigVMFunctionLibrary {links [], nodes []}
 RigVMModel {nodes [
   n1 = &RigUnit_PrepareForExecution()
   RigUnit_GetTransform = &RigUnit_GetTransform(Item=(Type=Bone,Name="lowerarm_l"), Item.Name=RigVMVariableNode() {variable=temp_bone_control_L}.Value)
   SetControlOffset = /Script/ControlRig.RigUnit_SetControlOffset(Control=L_Control, Space=GlobalSpace, ExecutePin=@n1.ExecutePin, Offset=@RigUnit_GetTransform.Transform) {is=struct RigVMUnitNode, template="SetControlOffset::Execute(in Control,io Offset,in Space)"}
   > &RigUnit_SetControlOffset(
    Control=R_Control
    Space=GlobalSpace
    Offset=&RigUnit_GetTransform(Item=(Type=Bone,Name="lowerarm_l"), Item.Name=RigVMVariableNode() {variable=temp_bone_control_R}.Value).Transform
   ) {template="SetControlOffset::Execute(in Control,io Offset,in Space)"}
   RigUnit_BeginExecution = &RigUnit_BeginExecution()
   &RigUnit_SetRotation(
    Item=(Type=Bone,Name="repeat_upperarm_r")
    Weight=1
    bInitial=False
    ExecutePin=@RigUnit_BeginExecution.ExecutePin
    Value=&RigUnit_GetTransform(Item=(Type=Bone,Name="upperarm_r"), Space=LocalSpace).Transform.Rotation
   )
   > RigUnit_SetRotation = &RigUnit_SetRotation(
    Item=(Type=Bone,Name="repeat_upperarm_l")
    Weight=1
    bInitial=False
    Value=&RigUnit_GetTransform(Item=(Type=Bone,Name="upperarm_l"), Space=LocalSpace).Transform.Rotation
   )
   AlphaInterp = &RigVMFunction_AlphaInterp(
    InterpSpeedDecreasing=3
    InterpSpeedIncreasing=5
    Scale=1
    bClampResult=False
    Value=RigVMVariableNode() {variable=Parachute_open_float}.Value
   )
   > /Script/ControlRig.RigUnit_SetRelativeTransformForItem(
    Child=(Type=Bone,Name="L_shoulder_pad")
    Parent=(Type=Bone,Name="L_shoulder_pad")
    Value=(Rotation=(X=0,Y=0,Z=0,W=1),Translation=(X=2,Y=0,Z=1),Scale3D=(X=0.9,Y=0.9,Z=0.9))
    bParentInitial=False
    bPropagateToChildren=True
    Weight=@AlphaInterp.Result
    Value.Rotation=&RigVMFunction_MathQuaternionFromEuler(Euler=(X=0,Y=0,Z=0)).Result
   ) {
    is=struct RigVMUnitNode
    template="Set Relative Transform::Execute(in Child,in Parent,in bParentInitial,in Value,in Weight,in bPropagateToChildren)"
   }
   > &RigUnit_SetRelativeTransformForItem(
    Child=(Type=Bone,Name="R_shoulder_pad")
    Parent=(Type=Bone,Name="R_shoulder_pad")
    Value=(Rotation=(X=0,Y=0,Z=0,W=1),Translation=(X=-2,Y=0,Z=-1),Scale3D=(X=0.9,Y=0.9,Z=0.9))
    bParentInitial=False
    bPropagateToChildren=True
    Weight=@AlphaInterp.Result
    Value.Rotation=&RigVMFunction_MathQuaternionFromEuler(Euler=(X=0,Y=0,Z=0)).Result
   ) {template="Set Relative Transform::Execute(in Child,in Parent,in bParentInitial,in Value,in Weight,in bPropagateToChildren)"}
   n2 = &RigVMFunction_AlphaInterp(InterpSpeedDecreasing=3, InterpSpeedIncreasing=5, Scale=1, bClampResult=False, Value=RigVMVariableNode() {variable=skyDiving_float}.Value)
   > &RigUnit_SetRelativeTransformForItem(
    Child=(Type=Bone,Name="L_shoulder_pad")
    Parent=(Type=Bone,Name="L_shoulder_pad")
    Value=(Rotation=(X=0,Y=0,Z=0,W=1),Translation=(X=5,Y=0,Z=1),Scale3D=(X=0.9,Y=0.9,Z=0.9))
    bParentInitial=False
    bPropagateToChildren=True
    Weight=@n2.Result
    Value.Rotation=&RigVMFunction_MathQuaternionFromEuler(Euler=(X=0,Y=0,Z=0)).Result
   ) {template="Set Relative Transform::Execute(in Child,in Parent,in bParentInitial,in Value,in Weight,in bPropagateToChildren)"}
   > &RigUnit_SetRelativeTransformForItem(
    Child=(Type=Bone,Name="R_shoulder_pad")
    Parent=(Type=Bone,Name="R_shoulder_pad")
    Value=(Rotation=(X=0,Y=0,Z=0,W=1),Translation=(X=-5,Y=0,Z=-1),Scale3D=(X=0.9,Y=0.9,Z=0.9))
    bParentInitial=False
    bPropagateToChildren=True
    Weight=@n2.Result
    Value.Rotation=&RigVMFunction_MathQuaternionFromEuler(Euler=(X=0,Y=0,Z=0)).Result
   ) {template="Set Relative Transform::Execute(in Child,in Parent,in bParentInitial,in Value,in Weight,in bPropagateToChildren)"}
   > n3 = &RigUnit_SphericalPoseReader(
    ActiveRegionScaleFactors=(PositiveWidth=2,NegativeWidth=2,PositiveHeight=2,NegativeHeight=2)
    ActiveRegionSize=0.05
    Debug=(bDrawDebug=false,bDraw2D=false,DebugScale=25,DebugSegments=20,DebugThickness=0.25)
    DriverAxis=(X=1,Y=0,Z=0)
    DriverItem=(Type=Bone,Name="upperarm_l")
    FalloffRegionScaleFactors=(PositiveWidth=1,NegativeWidth=1,PositiveHeight=1,NegativeHeight=1)
    FalloffSize=0.2
    OptionalParentItem=(Type=Bone,Name="clavicle_l")
    RotationOffset=(X=-10,Y=-30,Z=0)
   )
   n4 = RigVMVariableNode() {variable=is_melee}
   > &RigUnit_SetRelativeTransformForItem(
    Child=(Type=Bone,Name="L_shoulder_pad")
    Parent=(Type=Bone,Name="L_shoulder_pad")
    Value=(Rotation=(X=0,Y=0,Z=0,W=1),Translation=(X=5,Y=0,Z=-1),Scale3D=(X=1,Y=1,Z=1))
    bParentInitial=False
    bPropagateToChildren=True
    Value.Rotation=&RigVMFunction_MathQuaternionFromEuler(Euler=(X=20,Y=-2,Z=0)).Result
    Weight=&RigVMFunction_MathDoubleMul(B=&RigVMFunction_MathBoolToFloat(Value=@n4.Value).Result, A=@n3.OutputParam).Result
   ) {template="Set Relative Transform::Execute(in Child,in Parent,in bParentInitial,in Value,in Weight,in bPropagateToChildren)"}
   > n5 = &RigUnit_SphericalPoseReader(
    ActiveRegionScaleFactors=(PositiveWidth=2,NegativeWidth=2,PositiveHeight=2,NegativeHeight=2)
    ActiveRegionSize=0.03
    Debug=(bDrawDebug=false,bDraw2D=false,DebugScale=25,DebugSegments=20,DebugThickness=0.25)
    DriverAxis=(X=-1,Y=0,Z=0)
    DriverItem=(Type=Bone,Name="upperarm_r")
    FalloffRegionScaleFactors=(PositiveWidth=1,NegativeWidth=1,PositiveHeight=1,NegativeHeight=1)
    FalloffSize=0.2
    OptionalParentItem=(Type=Bone,Name="clavicle_r")
    RotationOffset=(X=-10,Y=145,Z=0)
   )
   > &RigUnit_SetRelativeTransformForItem(
    Child=(Type=Bone,Name="R_shoulder_pad")
    Parent=(Type=Bone,Name="R_shoulder_pad")
    Value=(Rotation=(X=0,Y=0,Z=0,W=1),Translation=(X=-5,Y=0,Z=1),Scale3D=(X=1,Y=1,Z=1))
    bParentInitial=False
    bPropagateToChildren=True
    Value.Rotation=&RigVMFunction_MathQuaternionFromEuler(Euler=(X=30,Y=-10,Z=0)).Result
    Weight=@n5.OutputParam
   ) {template="Set Relative Transform::Execute(in Child,in Parent,in bParentInitial,in Value,in Weight,in bPropagateToChildren)"}
   > n6 = &RigUnit_SphericalPoseReader(
    ActiveRegionScaleFactors=(PositiveWidth=2,NegativeWidth=2,PositiveHeight=2,NegativeHeight=2)
    ActiveRegionSize=0.03
    Debug=(bDrawDebug=true,bDraw2D=false,DebugScale=25,DebugSegments=20,DebugThickness=0.25)
    DriverAxis=(X=-1,Y=-1,Z=-1)
    DriverItem=(Type=Bone,Name="upperarm_r")
    FalloffRegionScaleFactors=(PositiveWidth=1,NegativeWidth=1,PositiveHeight=1,NegativeHeight=1)
    FalloffSize=0.2
    OptionalParentItem=(Type=Bone,Name="clavicle_r")
    RotationOffset=(X=-90,Y=145,Z=0)
   )
   > &RigUnit_SetRelativeTransformForItem(
    Child=(Type=Bone,Name="R_shoulder_pad")
    Parent=(Type=Bone,Name="R_shoulder_pad")
    Value=(Rotation=(X=0,Y=0,Z=0,W=1),Translation=(X=-5,Y=0,Z=1),Scale3D=(X=1,Y=1,Z=1))
    bParentInitial=False
    bPropagateToChildren=True
    Value.Rotation=&RigVMFunction_MathQuaternionFromEuler(Euler=(X=40,Y=-10,Z=0)).Result
    Weight=@n6.OutputParam
   ) {template="Set Relative Transform::Execute(in Child,in Parent,in bParentInitial,in Value,in Weight,in bPropagateToChildren)"}
   > n7 = &RigUnit_SphericalPoseReader(
    ActiveRegionScaleFactors=(PositiveWidth=2,NegativeWidth=2,PositiveHeight=2,NegativeHeight=2)
    ActiveRegionSize=0.03
    Debug=(bDrawDebug=true,bDraw2D=false,DebugScale=25,DebugSegments=20,DebugThickness=0.25)
    DriverAxis=(X=1,Y=3.3,Z=1.5)
    DriverItem=(Type=Bone,Name="upperarm_l")
    FalloffRegionScaleFactors=(PositiveWidth=1,NegativeWidth=1,PositiveHeight=1,NegativeHeight=1)
    FalloffSize=0.2
    OptionalParentItem=(Type=Bone,Name="clavicle_r")
    RotationOffset=(X=90,Y=145,Z=0)
   )
   n8 = &RigVMFunction_AlphaInterp(InterpSpeedDecreasing=10, InterpSpeedIncreasing=10, Scale=1, bClampResult=False, Value=@n7.OutputParam)
   RigVMFunction_MathBoolNot = &RigVMFunction_MathBoolNot(Value=RigVMVariableNode() {variable=in_FE}.Value)
   > &RigUnit_SetRelativeTransformForItem(
    Child=(Type=Bone,Name="L_shoulder_pad")
    Parent=(Type=Bone,Name="L_shoulder_pad")
    Value=(Rotation=(X=0,Y=0,Z=0,W=1),Translation=(X=0,Y=0,Z=-1),Scale3D=(X=1,Y=1,Z=1))
    bParentInitial=False
    bPropagateToChildren=True
    Value.Rotation=&RigVMFunction_MathQuaternionFromEuler(Euler=(X=30,Y=-15,Z=0)).Result
    Weight=&RigVMFunction_MathDoubleMul(A=@n8.Result, B=&RigVMFunction_MathBoolToFloat(Value=@RigVMFunction_MathBoolNot.Result).Result).Result
   ) {template="Set Relative Transform::Execute(in Child,in Parent,in bParentInitial,in Value,in Weight,in bPropagateToChildren)"}
   RigUnit_SetTransform = &RigUnit_SetTransform(
    ExecutePin=()
    Item=(Type=Bone,Name="lowerarm_l")
    Weight=1
    bInitial=False
    Value=&RigUnit_GetTransform(Item=(Type=Control,Name="L_Control")).Transform
    Item.Name=RigVMVariableNode() {variable=temp_bone_control_L}.Value
   )
   > &RigUnit_SetTransform(
    Item=(Type=Bone,Name="lowerarm_l")
    Weight=1
    bInitial=False
    Value=&RigUnit_GetTransform(Item=(Type=Control,Name="R_Control")).Transform
    Item.Name=RigVMVariableNode() {variable=temp_bone_control_R}.Value
   )
  ]}
]
hierarchy [
 root {loc=<0 0 0>, rot=<0 0 0 1>}
 pelvis {parent=root, loc=<0 -0.7321 92.4218>, rot=<0.0127 0.707 -0.0127 -0.707>}
 spine_01 {parent=pelvis, loc=<5.6002 -0.5128 0>, rot=<0 0 0.034 -0.9994>}
 spine_02 {parent=spine_01, loc=<8.0753 0.0139 0>, rot=<0 0 0.0238 -0.9997>}
 spine_03 {parent=spine_02, loc=<7.0393 0 0>, rot=<0 0 0.0149 -0.9999>}
 spine_04 {parent=spine_03, loc=<8.1024 0 0>, rot=<0 0 -0.0132 -0.9999>}
 spine_05 {parent=spine_04, loc=<10.0146 0 0>, rot=<0 0 -0.0015 -1>}
 clavicle_l {parent=spine_05, loc=<2.5856 -0.2866 -4.3777>, rot=<0.0465 -0.7289 -0.009 -0.6829>}
 upperarm_l {parent=clavicle_l, loc=<18.2264 0.1342 -0.001>, rot=<0.0253 -0.2693 0.0013 -0.9627>}
 lowerarm_l {parent=upperarm_l, loc=<29.0071 0.1686 -1.239>, rot=<0.0391 -0.0099 -0.2453 -0.9686>}
 hand_l {parent=lowerarm_l, loc=<25.8607 0.222 1.0274>, rot=<-0.7193 -0.0988 0.1193 -0.6773>}
 thumb_01_l {parent=hand_l, loc=<1.3118 -0.3191 -2.459>, rot=<0.6546 -0.4042 0.0664 -0.6354>}
 thumb_02_l {parent=thumb_01_l, loc=<5.9336 -0.0448 0.0401>, rot=<-0.0037 0.0213 0.298 -0.9543>}
 thumb_03_l {parent=thumb_02_l, loc=<2.2223 0.3682 -0.0907>, rot=<0.0118 -0.013 0.0066 -0.9998>}
 index_metacarpal_l {parent=hand_l, loc=<1.6535 0.0183 -2.96>, rot=<0.0158 -0.0124 -0.0173 -0.9996>}
 index_01_l {parent=index_metacarpal_l, loc=<8.3008 0.0357 -0.0524>, rot=<-0.0044 0.0485 0.0924 -0.9945>}
 index_02_l {parent=index_01_l, loc=<3.3587 -0.3004 0.0452>, rot=<-0.002 0.0027 0.3225 -0.9466>}
 index_03_l {parent=index_02_l, loc=<2.5578 -0.123 -0.0252>, rot=<-0.002 0.0067 0.2558 -0.9667>}
 middle_metacarpal_l {parent=hand_l, loc=<2.1856 -0.1725 -0.5984>, rot=<0.006 0.0168 -0.0329 -0.9993>}
 middle_01_l {parent=middle_metacarpal_l, loc=<7.3275 -0.2447 0.0129>, rot=<-0.0208 0.0042 0.0858 -0.9961>}
 middle_02_l {parent=middle_01_l, loc=<4.0123 0.1401 0.1175>, rot=<-0.0218 0.0258 0.3512 -0.9357>}
 middle_03_l {parent=middle_02_l, loc=<2.6077 0.0534 0.1373>, rot=<0.0082 0.0188 0.1708 -0.9851>}
 ring_metacarpal_l {parent=hand_l, loc=<2.2074 -0.3625 1.3638>, rot=<-0.0284 0.0722 -0.0458 -0.9959>}
 ring_01_l {parent=ring_metacarpal_l, loc=<7.0295 -0.6197 -0.1446>, rot=<-0.0489 -0.0252 0.1892 -0.9804>}
 ring_02_l {parent=ring_01_l, loc=<3.6174 0.4647 -0.1027>, rot=<0 0 0.2806 -0.9598>}
 ring_03_l {parent=ring_02_l, loc=<2.4152 0.114 0.0199>, rot=<0 0 0.1189 -0.9929>}
 pinky_metacarpal_l {parent=hand_l, loc=<1.7616 -0.9753 3.3814>, rot=<-0.0397 0.0814 -0.069 -0.9935>}
 pinky_01_l {parent=pinky_metacarpal_l, loc=<7.1709 -0.4831 -0.1094>, rot=<-0.0654 -0.0035 0.2242 -0.9723>}
 pinky_02_l {parent=pinky_01_l, loc=<2.8274 0.0893 -0.1043>, rot=<0.0037 0.0011 0.2854 -0.9584>}
 pinky_03_l {parent=pinky_02_l, loc=<2.2046 0.0307 -0.0078>, rot=<-0.0002 0.0098 0.1445 -0.9895>}
 lowerarm_twist_02_l {parent=lowerarm_l, loc=<9.6199 0 0>, rot=<0 0 0 1>}
 lowerarm_twist_01_l {parent=lowerarm_l, loc=<17.8549 0 0>, rot=<0 0 0 1>}
 upperarm_twist_01_l {parent=upperarm_l, loc=<7.9796 0.0843 -0.6195>, rot=<0 0 0 1>}
 upperarm_twist_02_l {parent=upperarm_l, loc=<19.5059 0.0843 -0.6195>, rot=<0 0 0 1>}
 repeat_upperarm_l {parent=clavicle_l, loc=<17.9725 -3.141 1.4798>, rot=<0.0253 -0.2693 0.0013 -0.9627>}
 L_shoulder_pad {parent=repeat_upperarm_l, loc=<0 0 0>, rot=<0 0 0 1>}
 attach_cape {parent=spine_05, loc=<8.8845 -10.0498 -0.0001>, rot=<0.0616 -0.7044 -0.0616 -0.7044>}
 attach_backpack {parent=spine_05, loc=<-0.1211 -11.8038 -0.0001>, rot=<-0.0093 -0.707 0.0093 -0.707>}
 neck_01 {parent=spine_05, loc=<10.815 1.9196 0>, rot=<0 0 -0.1864 -0.9825>}
 neck_02 {parent=neck_01, loc=<4.7347 0.3045 0>, rot=<0 0 0.0363 -0.9993>}
 head {parent=neck_02, loc=<5.7632 0.5552 0>, rot=<0 0 0.115 -0.9934>}
 clavicle_r {parent=spine_05, loc=<2.5855 -0.2866 4.3777>, rot=<0.7289 0.0465 -0.6829 0.009>}
 upperarm_r {parent=clavicle_r, loc=<-18.2264 -0.1342 0.001>, rot=<0.0253 -0.2693 0.0013 -0.9627>}
 lowerarm_r {parent=upperarm_r, loc=<-29.0071 -0.1686 1.239>, rot=<0.0391 -0.0099 -0.2453 -0.9686>}
 hand_r {parent=lowerarm_r, loc=<-25.8607 -0.222 -1.0274>, rot=<-0.7193 -0.0988 0.1193 -0.6773>}
 thumb_01_r {parent=hand_r, loc=<-1.3118 0.3191 2.459>, rot=<0.6546 -0.4042 0.0664 -0.6354>}
 thumb_02_r {parent=thumb_01_r, loc=<-5.9336 0.0448 -0.0401>, rot=<-0.0037 0.0213 0.298 -0.9543>}
 thumb_03_r {parent=thumb_02_r, loc=<-2.2223 -0.3682 0.0907>, rot=<0.0118 -0.013 0.0066 -0.9998>}
 index_metacarpal_r {parent=hand_r, loc=<-1.6535 -0.0183 2.96>, rot=<0.0158 -0.0124 -0.0173 -0.9996>}
 index_01_r {parent=index_metacarpal_r, loc=<-8.3008 -0.0357 0.0524>, rot=<-0.0044 0.0485 0.0924 -0.9945>}
 index_02_r {parent=index_01_r, loc=<-3.3587 0.3004 -0.0452>, rot=<-0.002 0.0027 0.3225 -0.9466>}
 index_03_r {parent=index_02_r, loc=<-2.5578 0.123 0.0252>, rot=<-0.002 0.0067 0.2558 -0.9667>}
 middle_metacarpal_r {parent=hand_r, loc=<-2.1856 0.1725 0.5984>, rot=<0.006 0.0168 -0.0329 -0.9993>}
 middle_01_r {parent=middle_metacarpal_r, loc=<-7.3275 0.2447 -0.0129>, rot=<-0.0208 0.0042 0.0858 -0.9961>}
 middle_02_r {parent=middle_01_r, loc=<-4.0123 -0.1401 -0.1175>, rot=<-0.0218 0.0258 0.3512 -0.9357>}
 middle_03_r {parent=middle_02_r, loc=<-2.6077 -0.0534 -0.1373>, rot=<0.0082 0.0188 0.1708 -0.9851>}
 ring_metacarpal_r {parent=hand_r, loc=<-2.2074 0.3625 -1.3638>, rot=<-0.0284 0.0722 -0.0458 -0.9959>}
 ring_01_r {parent=ring_metacarpal_r, loc=<-7.0295 0.6197 0.1446>, rot=<-0.0489 -0.0252 0.1892 -0.9804>}
 ring_02_r {parent=ring_01_r, loc=<-3.6174 -0.4647 0.1027>, rot=<0 0 0.2806 -0.9598>}
 ring_03_r {parent=ring_02_r, loc=<-2.4152 -0.114 -0.0199>, rot=<0 0 0.1189 -0.9929>}
 pinky_metacarpal_r {parent=hand_r, loc=<-1.7616 0.9753 -3.3814>, rot=<-0.0397 0.0814 -0.069 -0.9935>}
 pinky_01_r {parent=pinky_metacarpal_r, loc=<-7.1709 0.4831 0.1094>, rot=<-0.0654 -0.0035 0.2242 -0.9723>}
 pinky_02_r {parent=pinky_01_r, loc=<-2.8274 -0.0893 0.1043>, rot=<0.0037 0.0011 0.2854 -0.9584>}
 pinky_03_r {parent=pinky_02_r, loc=<-2.2046 -0.0307 0.0078>, rot=<-0.0002 0.0098 0.1445 -0.9895>}
 lowerarm_twist_02_r {parent=lowerarm_r, loc=<-9.6199 0 0>, rot=<0 0 0 1>}
 lowerarm_twist_01_r {parent=lowerarm_r, loc=<-17.8549 0 0>, rot=<0 0 0 1>}
 upperarm_twist_01_r {parent=upperarm_r, loc=<-7.9796 -0.0843 0.6195>, rot=<0 0 0 1>}
 upperarm_twist_02_r {parent=upperarm_r, loc=<-19.5059 -0.0843 0.6195>, rot=<0 0 0 1>}
 repeat_upperarm_r {parent=clavicle_r, loc=<-17.9725 3.141 -1.48>, rot=<0.0253 -0.2693 0.0013 -0.9627>}
 R_shoulder_pad {parent=repeat_upperarm_r, loc=<0 0 0>, rot=<0 0 0 1>}
 thigh_l {parent=pelvis, loc=<-1.8716 -0.0109 -11.9936>, rot=<0.0012 0.0572 0.0155 -0.9982>}
 calf_l {parent=thigh_l, loc=<-41.9178 -0.2246 0.0926>, rot=<0.0006 -0.0161 -0.0391 -0.9991>}
 foot_l {parent=calf_l, loc=<-38.7057 1.2438 0.2452>, rot=<0.0975 -0.0474 0.0317 -0.9936>}
 ball_l {parent=foot_l, loc=<-6.1614 13.1469 0.6807>, rot=<0.0001 -0.0001 -0.7063 -0.7079>}
 calf_twist_02_l {parent=calf_l, loc=<-15.9268 -0.0967 0.0397>, rot=<0.0237 0.0095 -0.0035 -0.9997>}
 thigh_twist_01_l {parent=thigh_l, loc=<-21.6303 0 0>, rot=<0.0013 0 0 -1>}
 thigh_r {parent=pelvis, loc=<-1.8716 -0.0109 11.9936>, rot=<0.0572 -0.0012 0.9982 0.0155>}
 calf_r {parent=thigh_r, loc=<41.9178 0.2246 -0.0926>, rot=<0.0006 -0.0161 -0.0391 -0.9991>}
 foot_r {parent=calf_r, loc=<38.7057 -1.2438 -0.2452>, rot=<0.0975 -0.0474 0.0317 -0.9936>}
 ball_r {parent=foot_r, loc=<6.1614 -13.1469 -0.6807>, rot=<0.0001 -0.0001 -0.7063 -0.7079>}
 calf_twist_02_r {parent=calf_r, loc=<15.9268 0.0967 -0.0397>, rot=<0.0237 0.0095 -0.0035 -0.9997>}
 thigh_twist_01_r {parent=thigh_r, loc=<21.6303 0 0>, rot=<0.0013 0 0 -1>}
 L_Control {controlType=EULER_TRANSFORM, type=Control, loc=<0 0 0>, rot=<0 0 0 1>}
 R_Control {controlType=EULER_TRANSFORM, type=Control, loc=<0 0 0>, rot=<0 0 0 1>}
]
variables [
 Parachute_open_float {type=double, default=0}
 in_FE {type=bool, default=False}
 is_melee {type=bool, default=False}
 skyDiving_float {type=double, default=0}
 temp_bone_control_L {type=FName, default=upperarm_l}
 temp_bone_control_R {type=FName, default=upperarm_r}
]
