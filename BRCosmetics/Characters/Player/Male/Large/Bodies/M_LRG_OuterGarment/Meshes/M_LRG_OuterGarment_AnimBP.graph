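# Animation Blueprint for the large male player OuterGarment body cosmetic, built on the custom character part anim instance.
# It copies the pose from the attached parent mesh and feeds it through the garment's Control Rig.
# Each update converts parachute-open and skydiving states to floats and checks whether the owner's weapon core animation is melee.
# These values and the in-front-end flag drive the rig's inputs.
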AnimBlueprint /BRCosmetics/Characters/Player/Male/Large/Bodies/M_LRG_OuterGarment/Meshes/M_LRG_OuterGarment_AnimBP
target_skeleton=&/M_LRG_OuterGarment_Skeleton
graphs [
 AnimGraph {nodes [
   n1 = Control Rig!(
    Source=Copy Pose From Mesh() {bUseAttachedParent=true}
    Alpha
    Parachute_open_float:double=Get Parachute_open_float() -> (Parachute_open_float:double)
    skyDiving_float:double=Get skyDiving_float() -> (skyDiving_float:double)
    is_melee:bool=Get is melee() -> (is melee:bool)
    in_FE:bool=Get bIsInFrontEnd()
   ) -> (Pose) {
    alphaBoolBlend {blendOption=Linear}
    controlRigAssetReference {blueprintRigClass=&/M_LRG_OuterGarment_CtrlRig.M_LRG_OuterGarment_CtrlRig_C}
   }
   Output Pose(Result=@n1) {layerGroup=None}
  ]}
 EventGraph {nodes [
   Event BlueprintUpdateAnimation()
   > Set Parachute_open_float(Parachute_open_float:double="To Float (Boolean)"(InBool=Get bIsParachuteOpened())) -> (Output_Get:double)
   > Set skyDiving_float(skyDiving_float:double="To Float (Boolean)"(InBool=Get bIsSkyDiving())) -> (Output_Get:double)
   n1 = GetOwnerAnimationInterface!(self:Custom Character Part Anim Instance Object Reference) -> (ReturnValue:Fort Player Animation Interface Object Reference)
   > Set is melee(is melee:bool="Equal (Enum)"(
     A:EFortWeaponCoreAnimation Enum=GetWeaponCoreAnim!(self:Fort Player Animation Interface Object Reference=@n1) -> (ReturnValue:EFortWeaponCoreAnimation Enum)
     B:EFortWeaponCoreAnimation Enum=Melee
    )) -> (Output_Get:bool)
  ]}
]
variables [
 is melee {type=boolean}
 Parachute_open_float {type=number}
 skyDiving_float {type=number}
]
