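# Physics asset for the large male player outer garment, providing body collision shapes that, by its name, serve as cape colliders.
# It places one capsule on each of the pelvis, upper spine bones, clavicles, arms, hands, thighs, calves and feet, with near-mirrored left and right sides and none on neck or head.
# Upper arms and calves use repeat and twist bones, and no constraints link the bodies.
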PhysicsAsset /BRCosmetics/Characters/Player/Male/Large/Bodies/M_LRG_OuterGarment/Meshes/M_LRG_OuterGarment_Physics_CapeColliders
bodies {
 calf_twist_02_l [{shapeType=Capsule, length=11.7105, radius=15.9101, center=<-3.4529 0.0401 -0.3587>, rotation=<rot -0.9992 73.5391 -88.2626>}]
 calf_twist_02_r [{shapeType=Capsule, length=11.7105, radius=15.9101, center=<3.4529 -0.0401 0.3587>, rotation=<rot -0.9992 73.5391 91.7374>}]
 clavicle_l [{shapeType=Capsule, length=11.5835, radius=15.2344, center=<22.2193 -0.3917 -1.0447>, rotation=<rot 15.4787 -81.463 -84.7679>}]
 clavicle_r [{shapeType=Capsule, length=11.5835, radius=15.2344, center=<-22.2193 0.3917 1.0447>, rotation=<rot 15.4787 -81.463 95.2321>}]
 foot_l [{shapeType=Capsule, length=19.6533, radius=9.7306, center=<-3.9301 3.5447 0.2358>, rotation=<rot 0 -174.2284 -78.467>}]
 foot_r [{shapeType=Capsule, length=19.8159, radius=9.8663, center=<3.8868 -3.6681 -0.3328>, rotation=<rot 0 -174.4751 -75.6283>}]
 hand_l [{shapeType=Capsule, length=9.1172, radius=7.3632, center=<7.2894 -2.3099 0.3797>, rotation=<rot 0 93.2606 -64.1396>}]
 hand_r [{shapeType=Capsule, length=9.206, radius=7.2352, center=<-7.2397 2.3174 -0.4093>, rotation=<rot 0 92.9549 -65.7533>}]
 lowerarm_l [{shapeType=Capsule, length=18.0138, radius=12.7859, center=<10.2081 -0.4216 1.6445>, rotation=<rot 0 -81.0645 -84.7691>}]
 lowerarm_r [{shapeType=Capsule, length=18.0138, radius=12.7859, center=<-10.2081 0.4216 -1.6445>, rotation=<rot 0 -81.0645 95.2309>}]
 pelvis [{shapeType=Capsule, length=17.9698, radius=14.0968, center=<-2.2329 -2.3664 0>}]
 repeat_upperarm_l [{shapeType=Capsule, length=18.9539, radius=12.3732, center=<16.257 -0.026 -0.3418>, rotation=<rot 0.2033 -83.9261 -84.2292>}]
 repeat_upperarm_r [{shapeType=Capsule, length=18.9539, radius=12.3732, center=<-16.257 0.026 0.3418>, rotation=<rot 0.2033 -83.9261 95.7708>}]
 spine_03 [{shapeType=Capsule, length=6.8065, radius=22.0967, center=<0 4.6524 0>}]
 spine_04 [{shapeType=Capsule, length=14.3908, radius=19.3727, shapeName=spine_04_capsule, rotation=<rot 0 -90 0>}]
 spine_05 [{shapeType=Capsule, length=18.5061, radius=22.2821, center=<0.7492 0.3055 0>}]
 thigh_l [{shapeType=Capsule, length=34.4672, radius=14.0298, center=<-24.9638 -0.8526 0.7832>, rotation=<rot 0 -81.6544 -88.2082>}]
 thigh_r [{shapeType=Capsule, length=34.4672, radius=14.0298, center=<24.9638 0.8526 -0.7832>, rotation=<rot 0 -81.6544 91.7918>}]
}
constraints []
